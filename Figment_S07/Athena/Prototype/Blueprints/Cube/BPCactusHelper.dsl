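; BPCactusHelper is a helper Blueprint in the Cube prototype that places a scorch decal on network clients.
; The RepNotify for SpawnDecalTransform spawns a non-emissive scorch decal at the replicated location and rotation.
; An authority switch limits the spawn to remote instances, so the authority spawns no decal here.
; The construction script is empty, and nothing in this file sets the transform.
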(fn OnRep_SpawnDecalTransform
  (Networking|SwitchHasAuthority
    (:Remote
      (Rendering|Decal|SpawnDecalatLocation "/FNGameplayCues/Athena/Prototype/Blueprints/Cube/FX/MD_Decal_Scorch_NoEmissive.MD_Decal_Scorch_NoEmissive" "64.000000,256.000000,256.000000" (.location (Variables|Default|GetSpawnDecalTransform)) (.rotation (Variables|Default|GetSpawnDecalTransform))))))

(fn ConstructionScript)
